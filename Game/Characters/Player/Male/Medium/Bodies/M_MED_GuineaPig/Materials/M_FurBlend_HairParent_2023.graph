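# Masked two-sided parent material for guinea pig character fur cards, blending fur on skinned meshes.
# Character PreFX feeds an optional furblend function, then attribute overrides, then PostFX.
# A mask texture and green vertex color shape opacity, with separate shadow-pass strength and zero at low quality; an optional shoe height mask is also defined.
# Vertex color or UV2 drives dithered pixel-depth-offset blending, while green vertex color scales normal-direction offset.
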Material /Game/Characters/Player/Male/Medium/Bodies/M_MED_GuineaPig/Materials/M_FurBlend_HairParent_2023
graph {
 properties {
  bCastRayTracedShadows=false
  bUseMaterialAttributes=true
  bUsedWithClothing=true
  bUsedWithMeshDeformer=true
  bUsedWithMorphTargets=true
  bUsedWithNanite=true
  bUsedWithSkeletalMesh=true
  blendMode=BLEND_Masked
  ditherOpacityMask=true
  pixelDepthOffsetMode=PDOM_Legacy
  shadingModel=MSM_FromMaterialExpression
  shadingRate=MSR_2x2
  twoSided=true
  usedShadingModels=Default Lit | Subsurface | Clear Coat | Subsurface Profile | Cloth
 }
 nodes [
  n1 = TextureSample(Tex=TextureObjectParameter() {group=Skin FX, parameterName=SkinFX_Mask}) {
   constCoordinate=1
   samplerType=SAMPLERTYPE_Masks
   texture=/Game/Characters/Player/Male/Medium/Bodies/M_MED_GuineaPig/Textures/T_GuineaPig_Body_M
  }
  n2 = MaterialFunctionCall!(R=@n1.R) {materialFunction=/Game/Global/MaterialFunctions/TechArt/MF_MaskedGeoHoloControl}
  n3 = Multiply(A=@n2.R Out, B=ScalarParameter() {defaultValue=1, group=GuineaPig_furblend, parameterName=FX_mask_intensity}) {constB=2}
  n4 = ScalarParameter() {defaultValue=2, group=GuineaPig_furblend, parameterName=fur cards Mask Threshold}
  n5 = VertexColor()
  n6 = Constant() {r=1}
  n7 = If(
   A=Power(Base=Saturate(None=@n3), Exp=LinearInterpolate(B=@n4, Alpha=@n5.G) {constA=1}) {constExponent=10}
   B=ScalarParameter() {defaultValue=0.5, group=GuineaPig_furblend, parameterName=mask clip value}
   "A > B"=@n6
   "A == B"=@n6
   "A < B"=@n2.R Out
  )
  n8 = Multiply(A=@n7, B=ScalarParameter() {group=GuineaPig_furblend, parameterName=shadow Pass intensity})
  n9 = TextureCoordinate() {coordinateIndex=2}
  n10 = Power(Base=StaticSwitchParameter(True=@n5.G, False=ComponentMask(None=@n9) {g=true}) {defaultValue=true, group=GuineaPig_furblend, parameterName=Use vertex colors}) {constExponent=1.5}
  n11 = LinearInterpolate(B=ScalarParameter() {defaultValue=0.5, group=GuineaPig_furblend, parameterName=FurBlend AlphaThreshold}, Alpha=@n10)
  n12 = Constant() {r=1}
  n13 = ScalarParameter() {defaultValue=1.5, group=GuineaPig_furblend, parameterName=FurBlendOpacityMask}
  n14 = ScalarParameter() {group=GuineaPig_furblend, parameterName=World Position Offset}
  n15 = VertexColor()
  n16 = MaterialFunctionCall() {materialFunction=&MF_CharacterParent_2023_PreFX}
  n17 = MaterialFunctionCall!(Mask Input=@n16.M, In=@n16.M) {materialFunction=&/MF_FN_GuineaPig_furblend}
  n18 = SetMaterialAttributes!(
   MaterialAttributes=GetMaterialAttributes(None=StaticSwitchParameter(True=@n17.Result, False=@n16.Result) {defaultValue=true, group=Skin FX, parameterName=UseFN_GuineaPig_furblend}) {attributeGetTypes [#1, #2]}.MaterialAttributes
   Opacity Mask=QualitySwitch(Default=ShadowReplace(Default=@n7, Shadow=@n8), Low=Constant())
   Pixel Depth Offset=Multiply(
    A=MaterialFunctionCall(Alpha Threshold=@n11, Random=@n12) {materialFunction=&DitherTemporalAA}.Result
    B=LinearInterpolate(B=@n13, Alpha=@n10)
   )
   Opacity=LinearInterpolate(A=Constant() {r=1}, Alpha=TwoSidedSign())
   Fuzz Color or Subsurface Color=Constant()
   World Position Offset=Multiply(A=VertexNormalWS(), B=Multiply(A=@n14, B=@n15.G))
  ) {attributeSetTypes [
    #1
    #3
    #4
    #5
    #6
   ]}
  n19 = GetMaterialAttributes(None=@n18) {attributeGetTypes [#1]}
  n20 = Multiply(A=@n19.Opacity Mask, B=ScalarParameter() {defaultValue=1, parameterName=OpacityMaskMultiplier})
  n21 = NamedRerouteUsage()
  MaterialFunctionCall(
   Material=SetMaterialAttributes!(MaterialAttributes=@n19.MaterialAttributes, Opacity Mask=Multiply(A=@n20, B=@n21)) {attributeSetTypes [#1]}
   M=@n16.M
   Pre Skinned Local Position=NamedRerouteDeclaration(None=@n16.Pre skinned Local Position) {name=PS-LocalPos, nodeColor=<rgba 0 0.3437 1 1>}
   Distance Blend=@n16.M
   Fresnel=@n16.M
  ) {materialFunction=&MF_CharacterParent_2023_PostFX}
  n22 = ScalarParameter() {defaultValue=25, group=Shoes, parameterName=Shoes_MaskHeight}
  n23 = Add(A=@n22, B=ScalarParameter() {defaultValue=15, group=Shoes, parameterName=Shoes_MaskWidth})
  n24 = ComponentMask(None=NamedRerouteUsage()) {b=true}
  NamedRerouteDeclaration(None=LinearInterpolate(B=SmoothStep(Min=@n22, Max=@n23, Value=@n24) {constMax=160}, Alpha=ScalarParameter() {group=Shoes, parameterName=UseShoes}) {constA=1}) {name=ShoesMaskMultiplier, nodeColor=<rgba 1 0 0.0241 1>}
 ]
}
