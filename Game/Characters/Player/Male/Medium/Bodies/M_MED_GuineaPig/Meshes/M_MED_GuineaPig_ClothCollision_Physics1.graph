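# Physics asset for the medium male GuineaPig player body mesh, named as a cloth collision proxy.
# It holds a single capsule body on the clavicle_r bone, offset and rotated from the bone origin.
# The capsule shape is named clavicle_l_capsule although it sits on the right clavicle.
# It has no constraints and no bodies on any other bone.
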PhysicsAsset /Game/Characters/Player/Male/Medium/Bodies/M_MED_GuineaPig/Meshes/M_MED_GuineaPig_ClothCollision_Physics1
bodies {clavicle_r [{
   shapeType=Capsule
   length=10.88
   radius=10.4711
   shapeName=clavicle_l_capsule
   center=<-6.8722 -0.909 1.8384>
   rotation=<rot 6.2227 84.6686 92.8584>
  }]}
constraints []
